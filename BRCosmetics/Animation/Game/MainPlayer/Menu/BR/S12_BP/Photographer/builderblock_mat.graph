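# Default Lit material named for a builder block in the Photographer folder of the Season 12 Battle Pass menu animation assets.
# Its graph holds a single vector parameter with a dark orange-brown default color.
# No textures, other expressions or links from that parameter to a material output appear in the graph.
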Material /BRCosmetics/Animation/Game/MainPlayer/Menu/BR/S12_BP/Photographer/builderblock_mat
graph {
 properties {pixelDepthOffsetMode=PDOM_Legacy, usedShadingModels=Default Lit}
 nodes [VectorParameter() {defaultValue=<rgba 0.1875 0.0508 0 0>}]
}
